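# Post-process AnimBP for the F_MED_ProngCling_Head female head mesh that adds secondary motion to both ears.
# Six chained single-bone AnimDynamics nodes simulate three bones per ear in root-relative space, with angular springs, limits and reduced gravity.
# Zipline enter/exit anim input events lower and restore a vector variable, but the dynamics nodes show fixed scales and never read it.
# The extra jump/move/sprint check is redundant because it also requires the zipline-enter event.
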AnimBlueprint /BRCosmetics/Characters/Player/Female/Medium/Heads/F_MED_ProngCling_Head/Meshes/F_MED_ProngCling_Head_AnimBP_PostProcess
target_skeleton=/Game/Characters/Player/Common/Fortnite_Base_Head/Fortnite_Base_Head_Export_Skeleton
graphs [
 AnimGraph {nodes [
   n1 = Anim Dynamics - Bone: dyn_ear03_r!(
    ComponentPose=Local To Component(LocalPose=Input Pose())
    Alpha=1
    ComponentLinearAccScale:Vector="0.5,0.5,0.5"
    ComponentLinearVelScale:Vector="0.5,0.5,0.5"
   ) -> (Pose) {
    angularDampingOverride=1
    angularSpringConstant=20000
    bAngularSpring=true
    bGravityOverrideInSimSpace=true
    bOverrideAngularDamping=true
    bUseGravityOverride=true
    gravityScale=0.35
    simulationSpace=RootRelative
    boundBone {boneName=dyn_ear03_r}
    componentAppliedLinearAccClamp=<3000 3000 3000>
    componentLinearAccScale=<0.5 0.5 0.5>
    componentLinearVelScale=<0.5 0.5 0.5>
    gravityOverride {z=-980}
    physicsBodyDefinitions [{
      boundBone {boneName=dyn_ear03_r}
      boxExtents=<40.25 40.25 40.25>
      constraintSetup {angularLimitsMax=<5 30 20>, angularLimitsMin=<-5 -30 -8>, angularTarget=<1 0 0>}
      localJointOffset=<5.1 0 0>
     }]
    simSpaceSettings {maxAngularAcceleration=35, maxAngularVelocity=7, simSpaceAngularAlpha=0.5}
   }
   n2 = Anim Dynamics - Bone: dyn_ear02_r!(ComponentPose=@n1, Alpha=1, ComponentLinearAccScale:Vector="0.5,0.5,0.5", ComponentLinearVelScale:Vector="0.5,0.5,0.5") -> (Pose) {
    angularDampingOverride=1
    angularSpringConstant=20000
    bAngularSpring=true
    bGravityOverrideInSimSpace=true
    bOverrideAngularDamping=true
    bUseGravityOverride=true
    gravityScale=0.35
    simulationSpace=RootRelative
    boundBone {boneName=dyn_ear02_r}
    componentAppliedLinearAccClamp=<3000 3000 3000>
    componentLinearAccScale=<0.5 0.5 0.5>
    componentLinearVelScale=<0.5 0.5 0.5>
    gravityOverride {z=-980}
    physicsBodyDefinitions [{
      boundBone {boneName=dyn_ear02_r}
      boxExtents=<40.25 40.25 40.25>
      constraintSetup {angularLimitsMax=<0 10 10>, angularLimitsMin=<0 -10 -10>, angularTarget=<1 0 0>}
      localJointOffset=<5.1 0 0>
     }]
    simSpaceSettings {maxAngularAcceleration=35, maxAngularVelocity=7, simSpaceAngularAlpha=0.5}
   }
   n3 = Anim Dynamics - Bone: dyn_ear01_r!(ComponentPose=@n2, Alpha=1, ComponentLinearAccScale:Vector="0.5,0.5,0.5", ComponentLinearVelScale:Vector="0.5,0.5,0.5") -> (Pose) {
    angularDampingOverride=1
    angularSpringConstant=20000
    bAngularSpring=true
    bGravityOverrideInSimSpace=true
    bOverrideAngularDamping=true
    bUseGravityOverride=true
    gravityScale=0.35
    simulationSpace=RootRelative
    boundBone {boneName=dyn_ear01_r}
    componentAppliedLinearAccClamp=<3000 3000 3000>
    componentLinearAccScale=<0.5 0.5 0.5>
    componentLinearVelScale=<0.5 0.5 0.5>
    gravityOverride {z=-980}
    physicsBodyDefinitions [{
      boundBone {boneName=dyn_ear01_r}
      boxExtents=<40.25 40.25 40.25>
      constraintSetup {angularLimitsMax=<5 25 10>, angularLimitsMin=<-5 -25 -8>, angularTarget=<1 0 0>}
      localJointOffset=<5.1 0 0>
     }]
    simSpaceSettings {maxAngularAcceleration=35, maxAngularVelocity=7, simSpaceAngularAlpha=0.5}
   }
   n4 = Anim Dynamics - Bone: dyn_ear03_l!(ComponentPose=@n3, Alpha=1, ComponentLinearAccScale:Vector="0.5,0.5,0.5", ComponentLinearVelScale:Vector="0.5,0.5,0.5") -> (Pose) {
    angularDampingOverride=1
    angularSpringConstant=17000
    bAngularSpring=true
    bGravityOverrideInSimSpace=true
    bOverrideAngularDamping=true
    bUseGravityOverride=true
    gravityScale=0.35
    simulationSpace=RootRelative
    boundBone {boneName=dyn_ear03_l}
    componentAppliedLinearAccClamp=<3000 3000 3000>
    componentLinearAccScale=<0.5 0.5 0.5>
    componentLinearVelScale=<0.5 0.5 0.5>
    gravityOverride {z=-980}
    physicsBodyDefinitions [{
      boundBone {boneName=dyn_ear03_l}
      boxExtents=<36.8 36.8 36.8>
      constraintSetup {angularLimitsMax=<5 -15 15>, angularLimitsMin=<-5 -15 -15>, angularTarget=<1 0 0>}
      localJointOffset=<5.1 0 0>
     }]
    simSpaceSettings {maxAngularAcceleration=35, maxAngularVelocity=7, simSpaceAngularAlpha=0.5}
   }
   n5 = Anim Dynamics - Bone: dyn_ear02_l!(ComponentPose=@n4, Alpha=1, ComponentLinearAccScale:Vector="0.5,0.5,0.5", ComponentLinearVelScale:Vector="0.5,0.5,0.5") -> (Pose) {
    angularDampingOverride=1
    angularSpringConstant=17000
    bAngularSpring=true
    bGravityOverrideInSimSpace=true
    bOverrideAngularDamping=true
    bUseGravityOverride=true
    gravityScale=0.35
    simulationSpace=RootRelative
    boundBone {boneName=dyn_ear02_l}
    componentAppliedLinearAccClamp=<3000 3000 3000>
    componentLinearAccScale=<0.5 0.5 0.5>
    componentLinearVelScale=<0.5 0.5 0.5>
    gravityOverride {z=-980}
    physicsBodyDefinitions [{
      boundBone {boneName=dyn_ear02_l}
      boxExtents=<36.8 36.8 36.8>
      constraintSetup {angularLimitsMax=<0 10 10>, angularLimitsMin=<0 -10 -10>, angularTarget=<1 0 0>}
      localJointOffset=<5.1 0 0>
     }]
    simSpaceSettings {maxAngularAcceleration=35, maxAngularVelocity=7, simSpaceAngularAlpha=0.5}
   }
   n6 = Anim Dynamics - Bone: dyn_ear01_l!(ComponentPose=@n5, Alpha=1, ComponentLinearAccScale:Vector="0.5,0.5,0.5", ComponentLinearVelScale:Vector="0.5,0.5,0.5") -> (Pose) {
    angularDampingOverride=1
    angularSpringConstant=17000
    bAngularSpring=true
    bGravityOverrideInSimSpace=true
    bOverrideAngularDamping=true
    bUseGravityOverride=true
    gravityScale=0.35
    simulationSpace=RootRelative
    boundBone {boneName=dyn_ear01_l}
    componentAppliedLinearAccClamp=<3000 3000 3000>
    componentLinearAccScale=<0.5 0.5 0.5>
    componentLinearVelScale=<0.5 0.5 0.5>
    gravityOverride {z=-980}
    physicsBodyDefinitions [{
      boundBone {boneName=dyn_ear01_l}
      boxExtents=<36.8 36.8 36.8>
      constraintSetup {angularLimitsMax=<5 15 5>, angularLimitsMin=<-5 -15 -5>, angularTarget=<1 0 0>}
      localJointOffset=<5.1 0 0>
     }]
    simSpaceSettings {maxAngularAcceleration=35, maxAngularVelocity=7, simSpaceAngularAlpha=0.5}
   }
   Output Pose(Result=Component To Local(ComponentPose=@n6))
  ]}
 EventGraph {nodes [
   n1 = Event BlueprintProcessAnimInputEvent()
   n2 = IsNamed(self=@n1.AnimInputEvent, InEventName=aie.Player.Zipline.Entered)
   n3 = OR Boolean(A=Get bIsMoving2D(), B=Get bIsTacticalSprinting())
   n4 = Get bIsJumping()
   > n5 = Branch(Condition=OR Boolean(A=@n2, B=AND Boolean!(A=@n2, B=@n4, C:bool=@n3) -> (ReturnValue)))
   > Set animDyn on zipline(animDyn on zipline:Vector="0.1,0.1,0.1") -> (Output_Get:Vector)
   Branch(execute=@n5.else, Condition=IsNamed(self=@n1.AnimInputEvent, InEventName=aie.Player.Zipline.Exited))
   > Set animDyn on zipline(animDyn on zipline="0.5,0.5,0.5")
   TryGetPawnOwner()
   Event BlueprintUpdateAnimation()
  ]}
]
variables [animDyn on zipline {type=Vector, default=<0.5 0.5 0.5>}]
